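# BlurSampleOffsets is a library material function that produces eight scaled 2D offsets, named and categorised for blur sampling and coordinate work.
# A single Vector2 OffsetAmount input, usable without a connection through its preview default, is multiplied by fixed direction vectors.
# The outputs cover four unit axis directions and four half-step diagonals, each named by its vector.
# The function does no texture sampling itself, only the offset math.
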MaterialFunction /Engine/Functions/Engine_MaterialFunctions01/ImageAdjustment/BlurSampleOffsets
graph {
 properties {bExposeToLibrary=true, description=Multiplies an input offset by a number of 2d vectors., libraryCategoriesText [Misc,Coordinates]}
 nodes [
  n1 = FunctionInput() {bUsePreviewValueAsDefault=true, inputName=OffsetAmount, inputType=FunctionInput_Vector2, previewValue=<1 1 0 1>}
  FunctionOutput(None=Multiply(A=Constant2Vector() {g=-1}, B=@n1)) {outputName="0,-1", sortPriority=2}
  FunctionOutput(None=Multiply(A=Constant2Vector() {g=0.5, r=0.5}, B=@n1)) {outputName=".5,.5", sortPriority=5}
  FunctionOutput(None=Multiply(A=Constant2Vector() {g=-0.5, r=-0.5}, B=@n1)) {outputName="-.5,-.5", sortPriority=7}
  FunctionOutput(None=Multiply(A=Constant2Vector() {g=0.5, r=-0.5}, B=@n1)) {outputName="-.5,.5", sortPriority=4}
  FunctionOutput(None=Multiply(A=Constant2Vector() {g=-0.5, r=0.5}, B=@n1)) {outputName=".5,-.5", sortPriority=6}
  FunctionOutput(None=Multiply(A=Constant2Vector() {r=1}, B=@n1)) {outputName="1,0", sortPriority=1}
  FunctionOutput(None=Multiply(A=Constant2Vector() {r=-1}, B=@n1)) {outputName="-1,0", sortPriority=3}
  FunctionOutput(None=Multiply(A=Constant2Vector() {g=1}, B=@n1)) {outputName="0,1"}
 ]
}
