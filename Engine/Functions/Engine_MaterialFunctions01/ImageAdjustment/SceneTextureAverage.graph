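# Material function, exposed to the library, that blurs scene color by averaging eight samples taken in a ring around the current pixel.
# The OffsetAmount input, a screen-fraction Vector2, feeds BlurSampleOffsets to produce four axial and four diagonal offsets.
# Each offset drives a SceneColor lookup, and the RGB results are summed and divided by eight for the "8 Samples" output, with no center sample.
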MaterialFunction /Engine/Functions/Engine_MaterialFunctions01/ImageAdjustment/SceneTextureAverage
graph {
 properties {bExposeToLibrary=true, description=Samples and averages the scene texture 8 times in a circular pattern.}
 nodes [
  n1 = FunctionInput() {
   bUsePreviewValueAsDefault=true
   description="Scene color offset fraction. "
   inputName=OffsetAmount
   inputType=FunctionInput_Vector2
   previewValue=<0.01 0 0 1>
  }
  n2 = MaterialFunctionCall!(OffsetAmount=@n1) {materialFunction=&/BlurSampleOffsets}
  n3 = Add(
   A=Add(
    A=SceneColor!(OffsetFraction=@n2."0,1") {inputMode=OffsetFraction}.RGB
    B=SceneColor(OffsetFraction=@n2."1,0") {inputMode=OffsetFraction}.RGB
   )
   B=Add(
    A=SceneColor(OffsetFraction=@n2."0,-1") {inputMode=OffsetFraction}.RGB
    B=SceneColor(OffsetFraction=@n2."-1,0") {inputMode=OffsetFraction}.RGB
   )
  )
  n4 = Add(
   A=Add(
    A=SceneColor(OffsetFraction=@n2."-.5,.5") {inputMode=OffsetFraction}.RGB
    B=SceneColor(OffsetFraction=@n2.".5,.5") {inputMode=OffsetFraction}.RGB
   )
   B=Add(
    A=SceneColor(OffsetFraction=@n2.".5,-.5") {inputMode=OffsetFraction}.RGB
    B=SceneColor(OffsetFraction=@n2."-.5,-.5") {inputMode=OffsetFraction}.RGB
   )
  )
  FunctionOutput(None=Divide(A=Add(A=@n3, B=@n4)) {constB=8}) {outputName=8 Samples}
 ]
}
