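# Material function in the Image Adjustment library that applies an unsharp-mask, high-pass style filter to caller-supplied samples.
# It outputs six axis-offset positions around Position (world position by default), built with BlurSampleOffsets, at which the caller evaluates the signal to filter.
# The returned samples are summed, divided by Number of Samples, and subtracted from the center, weighted by Amount.
# Lowering that count with axes left unconnected gives 2D or 1D filtering.
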MaterialFunction /Engine/Functions/Engine_MaterialFunctions01/ImageAdjustment/UnSharpMaskFunction
graph {
 properties {
  bExposeToLibrary=true
  description=Performs a high pass type operation on the input functions. This filters out low frequency information.
  libraryCategoriesText [Image Adjustment]
 }
 nodes [
  n1 = FunctionInput() {inputName=Center Sample, sortPriority=3}
  n2 = FunctionInput() {bUsePreviewValueAsDefault=true, inputName=Amount, inputType=FunctionInput_Scalar, sortPriority=1, previewValue=<1 0 0 1>}
  n3 = Add(B=@n2) {constA=1}
  n4 = Add(
   A=FunctionInput() {bUsePreviewValueAsDefault=true, inputName=Offset Sample +X, sortPriority=4}
   B=FunctionInput() {bUsePreviewValueAsDefault=true, inputName=Offset Sample -X, sortPriority=5}
  )
  n5 = Add(
   A=FunctionInput() {bUsePreviewValueAsDefault=true, inputName=Offset Sample +Y, sortPriority=6}
   B=FunctionInput() {bUsePreviewValueAsDefault=true, inputName=Offset Sample -Y, sortPriority=7}
  )
  n6 = FunctionInput() {bUsePreviewValueAsDefault=true, inputName=Offset Sample +Z, sortPriority=8}
  n7 = FunctionInput() {bUsePreviewValueAsDefault=true, inputName=Offset Sample -Z, sortPriority=9}
  n8 = FunctionInput() {
   bUsePreviewValueAsDefault=true
   description="Defaults to 6 samples to get 2 samples on each axis. If you want to skip a certain axis you can lower the number of samples to match to perform 2d or 1d highpass. For example, if you leave the two Z offsets disconnected and set this value to 4, it will do a 2D highpass on the XY plane."
   inputName=Number of Samples
   inputType=FunctionInput_Scalar
   sortPriority=10
   previewValue=<6 0 0 1>
  }
  n9 = Divide(A=Add(A=Add(A=@n4, B=@n5), B=Add(A=@n6, B=@n7)), B=@n8) {constB=4}
  FunctionOutput(None=Subtract(A=Multiply(A=@n1, B=@n3), B=Multiply(A=@n2, B=@n9))) {description=The high pass result.}
  n10 = FunctionInput(Preview=WorldPosition().XYZ) {bUsePreviewValueAsDefault=true, inputName=Position}
  n11 = FunctionInput() {bUsePreviewValueAsDefault=true, inputName=Sample Offset, sortPriority=2, previewValue=<0.01 0 0 1>}
  n12 = MaterialFunctionCall!(OffsetAmount=@n11) {materialFunction=&/BlurSampleOffsets}
  n13 = Constant()
  FunctionOutput(None=Add(A=@n10, B=AppendVector(A=@n12."0,-1", B=@n13))) {outputName=Offset -Y, sortPriority=4}
  n14 = Constant()
  FunctionOutput(None=Add(A=@n10, B=AppendVector(A=@n12."0,1", B=@n14))) {outputName=Offset +Y, sortPriority=3}
  n15 = Constant()
  FunctionOutput(None=Add(A=@n10, B=AppendVector(A=@n15, B=@n12."0,-1"))) {outputName=Offset -Z, sortPriority=6}
  n16 = Constant()
  FunctionOutput(None=Add(A=@n10, B=AppendVector(A=@n16, B=@n12."0,1"))) {outputName=Offset +Z, sortPriority=5}
  n17 = Constant()
  FunctionOutput(None=Add(A=@n10, B=AppendVector(A=@n12."1,0", B=@n17))) {outputName=Offset +X, sortPriority=1}
  n18 = Constant()
  FunctionOutput(None=Add(A=@n10, B=AppendVector(A=@n12."-1,0", B=@n18))) {outputName=Offset -X, sortPriority=2}
 ]
}
