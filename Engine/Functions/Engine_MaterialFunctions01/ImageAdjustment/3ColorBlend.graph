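# 3ColorBlend is a library material function that blends between three colors A, B and C based on a greyscale Alpha.
# Alpha is rescaled and clamped into two half-range factors, one driving an A-to-B lerp and the other B-to-C.
# A final lerp mixes both results by the raw Alpha, so B peaks at the midpoint and the falloff is curved.
# Inputs default to red, green, blue and a LinearGradient U ramp.
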MaterialFunction /Engine/Functions/Engine_MaterialFunctions01/ImageAdjustment/3ColorBlend
graph {
 properties {bExposeToLibrary=true, description=Blends between 3 colors based on a greyscale input, libraryCategoriesText [Misc, Image Adjustment]}
 nodes [
  n1 = FunctionInput() {bUsePreviewValueAsDefault=true, inputName=A, previewValue=<1 0 0 1>}
  n2 = FunctionInput() {bUsePreviewValueAsDefault=true, inputName=B, sortPriority=1, previewValue=<0 1 0 1>}
  n3 = FunctionInput(Preview=MaterialFunctionCall() {materialFunction=&LinearGradient}.UGradient) {bUsePreviewValueAsDefault=true, inputName=Alpha, inputType=FunctionInput_Scalar, sortPriority=4}
  n4 = Clamp(None=Multiply(A=@n3) {constB=2})
  n5 = FunctionInput() {bUsePreviewValueAsDefault=true, inputName=C, sortPriority=1, previewValue=<0 0 1 1>}
  n6 = Clamp(None=Multiply(A=Add(A=@n3) {constB=-0.5}) {constB=2})
  FunctionOutput(None=LinearInterpolate(A=LinearInterpolate(A=@n1, B=@n2, Alpha=@n4), B=LinearInterpolate(A=@n2, B=@n5, Alpha=@n6), Alpha=@n3) {desc="Blend it!"})
 ]
}
