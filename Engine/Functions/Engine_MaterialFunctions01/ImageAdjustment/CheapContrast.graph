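# CheapContrast is a material function in the Misc and Image Adjustment library categories that cheaply boosts contrast of a scalar, like pulling in Levels edges in Photoshop.
# The input value drives a lerp between negative Contrast and one plus Contrast, widening the range around the original.
# The result is clamped and masked to red, so zero Contrast applies no boost.
# An inverted LinearGradient feeds the value input's preview.
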MaterialFunction /Engine/Functions/Engine_MaterialFunctions01/ImageAdjustment/CheapContrast
graph {
 properties {
  bExposeToLibrary=true
  description=Cheaply adds contrast similar to pulling in edges in Photoshops Levels tool
  libraryCategoriesText [Misc, Image Adjustment]
 }
 nodes [
  n1 = FunctionInput() {bUsePreviewValueAsDefault=true, description=0 is default contrast, inputName=Contrast, inputType=FunctionInput_Scalar, sortPriority=1}
  n2 = MaterialFunctionCall() {materialFunction=&LinearGradient}
  n3 = LinearInterpolate(
   A=Subtract(B=@n1) {constA=0, desc=Low negative value}
   B=Add(A=@n1) {desc=High Positive}
   Alpha=FunctionInput(Preview=OneMinus(None=@n2.VGradient)) {inputType=FunctionInput_Scalar}
  ) {desc=Lerps between greater range of value}
  FunctionOutput(None=ComponentMask(None=Clamp(None=@n3)) {r=true})
 ]
}
