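# CheapContrast_RGB is a library material function that cheaply adds contrast to an RGB input, similar to pulling in the edges in a Photoshop Levels adjustment.
# The scalar Contrast input, neutral at 0, is negated for a low bound and offset upward for a high bound.
# The color drives a Lerp between these bounds, with the low one appended into a vector3, and a Clamp trims the widened range.
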MaterialFunction /Engine/Functions/Engine_MaterialFunctions01/ImageAdjustment/CheapContrast_RGB
graph {
 properties {
  bExposeToLibrary=true
  description=Cheaply adds contrast similar to pulling in edges in Photoshops Levels tool
  libraryCategoriesText [Misc, Image Adjustment]
 }
 nodes [
  n1 = FunctionInput() {bUsePreviewValueAsDefault=true, description=0 is default contrast, inputName=Contrast, inputType=FunctionInput_Scalar, sortPriority=1}
  n2 = Subtract(B=@n1) {constA=0, desc=Low negative value}
  n3 = AppendVector(A=AppendVector(A=@n2, B=@n2) {desc=need to make one of the Lerp inputs a vector3}, B=@n2)
  n4 = Add(A=@n1) {desc=High Positive}
  n5 = FunctionInput(Preview=TextureSample() {texture=/Engine/EngineDebugMaterials/HeatmapGradient}.RGB)
  FunctionOutput(None=Clamp(None=LinearInterpolate(A=@n3, B=@n4, Alpha=@n5) {desc=Lerps between greater range of value}))
 ]
}
