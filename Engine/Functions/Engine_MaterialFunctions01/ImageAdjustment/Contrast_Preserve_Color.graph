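# Material function in the Image Adjustment library that changes image contrast while preserving the relative color values.
# It fully desaturates the Color input to get luminosity, then raises that to the Contrast exponent.
# The result multiplies the normalized Color vector, so color direction and brightness are handled separately.
# Because normalization gives unit vector length rather than unit luminosity, a Contrast of 1 does not generally return the original color.
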MaterialFunction /Engine/Functions/Engine_MaterialFunctions01/ImageAdjustment/Contrast_Preserve_Color
graph {
 properties {
  bExposeToLibrary=true
  description=This function allows you to modify the contrast of an image without affecting the relative color values.
  libraryCategoriesText [Image Adjustment]
 }
 nodes [
  n1 = FunctionInput() {
   bUsePreviewValueAsDefault=true
   description=The input image. Luminosity will be calculated by desaturating this input.
   inputName=Color
   previewValue=<1 1 1 1>
  }
  n2 = Desaturation(None=@n1, Fraction=Constant() {r=1})
  n3 = FunctionInput() {
   bUsePreviewValueAsDefault=true
   description=This input is applied as an exponent to the luminosity only.
   inputName=Contrast
   inputType=FunctionInput_Scalar
   sortPriority=2
   previewValue=<1 0 0 1>
  }
  FunctionOutput(None=Multiply(A=Normalize(VectorInput=@n1), B=Power(Base=@n2, Exp=@n3)))
 ]
}
