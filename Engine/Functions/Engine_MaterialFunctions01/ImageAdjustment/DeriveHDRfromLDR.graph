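# Material function in the Image Adjustment library that derives an HDR range from an LDR color input while maintaining chroma.
# A luminance mask comes from desaturating the input, optionally after a 0.4545 gamma power, then applying an adjustable power.
# The input is expanded by an inverse tonemap-style divide, roughly 1/(1-luma), scaled by the dynamic range.
# An intensity multiplier scales the result, and a tint blends in by luminance to color mainly bright values.
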MaterialFunction /Engine/Functions/Engine_MaterialFunctions01/ImageAdjustment/DeriveHDRfromLDR
graph {
 properties {
  bExposeToLibrary=true
  description=Derive an HDR range from an LDR texture while maintaining chroma and minimizing artifacts
  libraryCategoriesText [Misc, Image Adjustment]
 }
 nodes [
  n1 = FunctionInput() {
   bUsePreviewValueAsDefault=true
   desc=Input an LDR texture or a material chain resulting in a Vector3
   description=Input an LDR texture or a material chain resulting in a Vector3.
   inputName=LDR Input
   previewValue=<0.25 0.25 0.25 1>
  }
  n2 = FunctionInput() {
   bUsePreviewValueAsDefault=true
   desc=The dynamic range of the HDR values derived from your LDR Input
   description=The dynamic range of the HDR values derived from your LDR Input
   inputName=HDR Dynamic Range
   inputType=FunctionInput_Scalar
   sortPriority=1
   previewValue=<100 0 0 1>
  }
  n3 = Power(Base=@n1) {constExponent=0.4545}
  n4 = FunctionInput(Preview=StaticBoolParameter() {group=Derive HDR from LDR, parameterName=Luminance from Linear}) {
   bUsePreviewValueAsDefault=true
   desc="This can create a softer result by deriving the Luminance from\r\nthe input in Linear space instead of SRGB "
   description=This can create a softer result by deriving the Luminance from the input in Linear space instead of SRGB
   inputName=Derive Luminance from Linear
   inputType=FunctionInput_StaticBool
   sortPriority=3
  }
  n5 = FunctionInput() {
   bUsePreviewValueAsDefault=true
   desc="Lower this value to derive more of the HDR values from the base input rather than its' Luminance. \r\nArtistic parameter, should probably stay around 1"
   description="Lower this value to derive more of the HDR values from the base input rather than its' Luminance. Artistic parameter, should probably stay around 1."
   inputName=Desaturate Luminance
   inputType=FunctionInput_Scalar
   sortPriority=3
   previewValue=<1 0 0 1>
  }
  n6 = FunctionInput() {
   bUsePreviewValueAsDefault=true
   desc="Power function applied to the Luminance which is used to derive the HDR range "
   description=Power function applied to the Luminance which is used to derive the HDR range
   inputName=Derive HDR Power
   inputType=FunctionInput_Scalar
   sortPriority=3
   previewValue=<1 0 0 1>
  }
  n7 = Power(
   Base=Desaturation(None=StaticSwitch(True=@n3, False=@n1, Value=@n4), Fraction=@n5) {desc=Luminence, luminanceFactors=<rgba 0.33 0.5 0.16 0>}
   Exp=@n6
  ) {constExponent=1}
  n8 = Add(A=@n2, B=Subtract(B=Multiply(A=@n2, B=@n7)))
  n9 = FunctionInput() {
   bUsePreviewValueAsDefault=true
   desc=A multiplier against the overall intensity of the HDR range
   description=A multiplier against the overall intensity of the HDR range
   inputName=HDR Intensity Multiplier
   inputType=FunctionInput_Scalar
   sortPriority=2
   previewValue=<1 0 0 1>
  }
  n10 = FunctionInput() {
   bUsePreviewValueAsDefault=true
   desc="Tint only the HDR range of the output "
   description=Tint only the HDR range of the output
   inputName=HDR Tint
   sortPriority=2
   previewValue=<1 1 1 1>
  }
  n11 = Multiply(A=@n9, B=LinearInterpolate(A=Constant() {r=1}, B=@n10, Alpha=@n7))
  FunctionOutput(None=Divide(A=Multiply(A=@n1, B=@n2), B=Divide(A=@n8, B=@n11)) {desc="Basic tonemap 1/(1-luma)"}) {outputName=RemappedRange}
 ]
}
