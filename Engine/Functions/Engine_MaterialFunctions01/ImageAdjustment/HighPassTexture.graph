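# HighPassTexture is a library material function that filters low-frequency information out of an input texture object.
# It averages samples offset around the UVs by a texel width via BlurSampleOffsets, then subtracts that blur from the center sample.
# A High Quality static switch adds four diagonal taps to the four axis taps.
# It outputs the raw difference and a version biased, raised to a Contrast power and halved.
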MaterialFunction /Engine/Functions/Engine_MaterialFunctions01/ImageAdjustment/HighPassTexture
graph {
 properties {
  bExposeToLibrary=true
  description=Performs a high pass type operation on an input texture object. This filters out low frequency information.
  libraryCategoriesText [Image Adjustment]
 }
 nodes [
  n1 = FunctionInput(Preview=TextureCoordinate()) {bUsePreviewValueAsDefault=true, inputName=UVs, inputType=FunctionInput_Vector2, sortPriority=1}
  n2 = FunctionInput() {inputName=Texture Object, inputType=FunctionInput_Texture2D}
  n3 = TextureSample(UVs=@n1, Tex=@n2) {desc=Center Sample, samplerType=SAMPLERTYPE_LinearColor}
  n4 = FunctionInput() {
   bUsePreviewValueAsDefault=true
   description="The width of the highpass offset, in texels."
   inputName=Highpass Offset
   inputType=FunctionInput_Scalar
   sortPriority=3
   previewValue=<8 0 0 1>
  }
  n5 = MaterialFunctionCall!(OffsetAmount=Divide(A=@n4, B=TextureProperty(Tex=@n2))) {materialFunction=&/BlurSampleOffsets}
  n6 = Add(A=@n1, B=@n5."0,1")
  n7 = Add(A=@n1, B=@n5."1,0")
  n8 = Add(A=@n1, B=@n5."0,-1")
  n9 = Add(A=@n1, B=@n5."-1,0")
  n10 = Add(
   A=Add(
    A=TextureSample(UVs=@n6, Tex=@n2) {samplerType=SAMPLERTYPE_LinearColor}.RGB
    B=TextureSample(UVs=@n7, Tex=@n2) {samplerType=SAMPLERTYPE_LinearColor}.RGB
   )
   B=Add(
    A=TextureSample(UVs=@n8, Tex=@n2) {samplerType=SAMPLERTYPE_LinearColor}.RGB
    B=TextureSample(UVs=@n9, Tex=@n2) {samplerType=SAMPLERTYPE_LinearColor}.RGB
   )
  )
  n11 = Add(A=@n1, B=@n5."-.5,.5")
  n12 = Add(A=@n1, B=@n5.".5,.5")
  n13 = Add(A=@n1, B=@n5.".5,-.5")
  n14 = Add(A=@n1, B=@n5."-.5,-.5")
  n15 = Add(
   A=Add(
    A=TextureSample(UVs=@n11, Tex=@n2) {samplerType=SAMPLERTYPE_LinearColor}.RGB
    B=TextureSample(UVs=@n12, Tex=@n2) {samplerType=SAMPLERTYPE_LinearColor}.RGB
   )
   B=Add(
    A=TextureSample(UVs=@n13, Tex=@n2) {samplerType=SAMPLERTYPE_LinearColor}.RGB
    B=TextureSample(UVs=@n14, Tex=@n2) {samplerType=SAMPLERTYPE_LinearColor}.RGB
   )
  )
  n16 = FunctionInput(Preview=StaticBool()) {
   bUsePreviewValueAsDefault=true
   description="When set to true, performs 8 sample instead of 4 to perform an average. Set to False by default."
   inputName=High Quality
   inputType=FunctionInput_StaticBool
   sortPriority=5
   previewValue=<0.01 0 0 1>
  }
  n17 = StaticSwitch(True=Divide(A=Add(A=@n10, B=@n15)) {constB=8}, False=Divide(A=@n10) {constB=4}, Value=@n16)
  n18 = Add(A=FunctionOutput(None=Subtract(A=@n3.RGB, B=@n17)) {description=The high pass result., outputName=Result No Contrast})
  n19 = FunctionInput() {bUsePreviewValueAsDefault=true, inputName=Contrast, inputType=FunctionInput_Scalar, sortPriority=2, previewValue=<1 0 0 1>}
  FunctionOutput(None=Multiply(A=Power(Base=@n18, Exp=@n19)) {constB=0.5}) {description=The high pass result., sortPriority=1}
 ]
}
