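# LinearTosRGB is a material function in the Color library category that converts a scalar linear value to sRGB.
# An If node compares the input to a threshold of about 0.0031 to pick between the two segments of the piecewise transfer curve.
# Values above it use a 1/2.4 power with scale and offset, while values at or below it use a linear 12.92 multiply.
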MaterialFunction /Engine/Functions/Engine_MaterialFunctions01/ImageAdjustment/LinearTosRGB
graph {
 properties {bExposeToLibrary=true, description=Convert Linear to sRGB, libraryCategoriesText [Color]}
 nodes [
  n1 = FunctionInput() {description=Linear Value, inputName=Linear, inputType=FunctionInput_Scalar}
  n2 = Multiply(A=Power(Base=@n1, Exp=Constant() {desc=1/2.4, r=0.4167}), B=Constant() {r=1.055})
  n3 = Constant() {r=-0.055}
  n4 = Multiply(A=@n1, B=Constant() {r=12.92})
  FunctionOutput(None=If(A=@n1, B=Constant() {desc=0.04045/12.92, r=0.0031}, "A > B"=Add(A=@n2, B=@n3), "A == B"=@n4, "A < B"=@n4)) {outputName=sRGB}
 ]
}
