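# Material function in the Image Adjustment library that merges an input color with an input luminosity value.
# It normalizes the Color vector, then multiplies it by the scalar Luminosity to produce the single output.
# Brightness is set by vector length rather than perceptual luminance weighting, and both inputs fall back to preview values when unconnected.
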MaterialFunction /Engine/Functions/Engine_MaterialFunctions01/ImageAdjustment/Luminosity_And_Color
graph {
 properties {
  bExposeToLibrary=true
  description=This function takes an input color and an input luminosity and merges them.
  libraryCategoriesText [Image Adjustment]
 }
 nodes [
  n1 = FunctionInput() {
   bUsePreviewValueAsDefault=true
   description=The color that will be applied. It is normalized inside of the function.
   inputName=Color
   previewValue=<1 1 1 1>
  }
  n2 = FunctionInput() {
   bUsePreviewValueAsDefault=true
   description=Luminosity used for result.
   inputName=Luminosity
   inputType=FunctionInput_Scalar
   sortPriority=1
   previewValue=<1 0 0 1>
  }
  FunctionOutput(None=Multiply(A=Normalize(VectorInput=@n1), B=@n2))
 ]
}
