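# SCurve is a library material function that applies an S-curve image adjustment, shaped by a scalar Power input.
# It remaps the lower half of the input with a power curve and the upper half with an inverted power curve, each scaled to half the range.
# A lerp joins the halves without branching, using the upper curve multiplied by a large constant and clamped as a near-hard switch at the midpoint.
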MaterialFunction /Engine/Functions/Engine_MaterialFunctions01/ImageAdjustment/SCurve
graph {
 properties {bExposeToLibrary=true, description=Applies an S curve image adjustment, libraryCategoriesText [Misc, Image Adjustment]}
 nodes [
  n1 = Constant() {r=2}
  n2 = FunctionInput(Preview=ComponentMask(None=TextureCoordinate()) {r=true})
  n3 = FunctionInput() {bUsePreviewValueAsDefault=true, inputName=Power, inputType=FunctionInput_Scalar, previewValue=<1 0 0 1>}
  n4 = Power(Base=Clamp(None=Multiply(A=@n1, B=@n2)), Exp=@n3)
  n5 = Constant() {r=0.5}
  n6 = Constant() {r=0.5}
  n7 = Constant() {r=-0.5}
  n8 = Clamp(None=Multiply(A=Add(A=@n2, B=@n7), B=Constant() {r=2}))
  n9 = OneMinus(None=Power(Base=OneMinus(None=@n8), Exp=@n3))
  n10 = Multiply(A=@n9, B=Constant() {r=0.5})
  n11 = Multiply(A=@n9, B=Constant() {r=500})
  FunctionOutput(None=LinearInterpolate(A=Multiply(A=@n4, B=@n5), B=Add(A=@n6, B=@n10), Alpha=Clamp(None=@n11)))
 ]
}
